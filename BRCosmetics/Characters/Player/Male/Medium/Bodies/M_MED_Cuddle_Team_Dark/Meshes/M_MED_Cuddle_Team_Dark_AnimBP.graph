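# Cosmetic AnimBP for the Cuddle Team Dark male body, adding secondary motion to accessories on a pose copied from the attached parent mesh.
# Weighted constraints place bandana, bowtie and shoulder-loop bones, while spring-based AnimDynamics sways those plus grenades and shoe tongues.
# Targeting events set target_alpha, driving interpolated additive offsets on both bowtie sides.
# Two alternate shoulder-loop constraints including upper-arm twist bones sit disconnected from the output.
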AnimBlueprint /BRCosmetics/Characters/Player/Male/Medium/Bodies/M_MED_Cuddle_Team_Dark/Meshes/M_MED_Cuddle_Team_Dark_AnimBP
target_skeleton=&/M_MED_Cuddle_Team_Dark_Skeleton
graphs [
 AnimGraph {nodes [
   n1 = "Constraint - shldrLoop_r_zero (PPPPPPP)"!(
    ComponentPose
    Alpha=1
    ConstraintWeights_0=0.1448
    ConstraintWeights_1:float=0.0132
    ConstraintWeights_2:float=0.0301
    ConstraintWeights_3:float=0.0932
    ConstraintWeights_4:float=0.013
    ConstraintWeights_5:float=0.1214
    ConstraintWeights_6:float=0.5842
   ) -> (Pose) {
    alphaBoolBlend {blendOption=Linear}
    boneToModify {boneName=shldrLoop_r_zero}
    constraintSetup [
     {targetBone {boneName=upperarm_twist_01_r}}
     {targetBone {boneName=spine_01}}
     {targetBone {boneName=spine_02}}
     {targetBone {boneName=spine_03}}
     {targetBone {boneName=spine_04}}
     {targetBone {boneName=spine_05}}
     {targetBone {boneName=clavicle_r}}
    ]
    constraintWeights [1,1,1,1,1,1,1]
   }
   "Constraint - shldrLoop_l_zero (PPPP)"!(
    ComponentPose=@n1
    Alpha=1
    ConstraintWeights_0=0.0695
    ConstraintWeights_1:float=0.1325
    ConstraintWeights_2:float=0.6393
    ConstraintWeights_3:float=0.1587
   ) -> (Pose) {
    alphaBoolBlend {blendOption=Linear}
    boneToModify {boneName=shldrLoop_l_zero}
    constraintSetup [
     {targetBone {boneName=spine_03}}
     {targetBone {boneName=spine_05}}
     {targetBone {boneName=clavicle_l}}
     {targetBone {boneName=upperarm_twist_01_l}}
    ]
    constraintWeights [1,1,1,1]
   }
   n2 = "Constraint - bicepBandana_r (PPPPP)"!(
    ComponentPose=Local To Component(LocalPose=Copy Pose From Mesh() {bUseAttachedParent=true})
    Alpha=1
    ConstraintWeights_0=0.0231
    ConstraintWeights_1:float=0.348
    ConstraintWeights_2:float=0.0128
    ConstraintWeights_3:float=0.3781
    ConstraintWeights_4:float=0.238
   ) -> (Pose) {
    alphaBoolBlend {blendOption=Linear}
    boneToModify {boneName=bicepBandana_r}
    constraintSetup [
     {targetBone {boneName=clavicle_r}}
     {targetBone {boneName=upperarm_r}}
     {targetBone {boneName=spine_03}}
     {targetBone {boneName=upperarm_twist_02_r}}
     {targetBone {boneName=upperarm_twist_01_r}}
    ]
    constraintWeights [1,1,1,1,1]
   }
   n3 = Anim Dynamics - Bone: bicepBandana_1_r(ComponentPose=@n2, Alpha=1, ExternalForce="0,0,0") {
    angularDampingOverride=1
    angularSpringConstant=1500
    bAngularSpring=true
    bGravityOverrideInSimSpace=true
    bOverrideAngularDamping=true
    gravityScale=0.35
    simulationSpace=RootRelative
    alphaBoolBlend {blendOption=Linear}
    alphaScaleBiasClamp {scale=0.5}
    boundBone {boneName=bicepBandana_1_r}
    physicsBodyDefinitions [{
      sphereCollisionRadius=0
      boundBone {boneName=bicepBandana_1_r}
      boxExtents=<10 20 20>
      constraintSetup {angularLimitsMax=<0 45 80>, angularLimitsMin=<0 -45 -80>, angularTarget=<1 0 0>}
      localJointOffset=<10 0 0>
     }]
   }
   n4 = Anim Dynamics - Bone: bicepBandana_2_r(ComponentPose=@n3, Alpha=1, ExternalForce="0,0,0") {
    angularDampingOverride=1
    angularSpringConstant=1500
    bAngularSpring=true
    bGravityOverrideInSimSpace=true
    bOverrideAngularDamping=true
    gravityScale=0.35
    simulationSpace=RootRelative
    alphaBoolBlend {blendOption=Linear}
    alphaScaleBiasClamp {scale=0.5}
    boundBone {boneName=bicepBandana_2_r}
    physicsBodyDefinitions [{
      sphereCollisionRadius=0
      boundBone {boneName=bicepBandana_2_r}
      boxExtents=<10 20 20>
      constraintSetup {angularLimitsMax=<0 45 80>, angularLimitsMin=<0 -45 -80>, angularTarget=<1 0 0>}
      localJointOffset=<10 0 0>
     }]
   }
   n5 = "Constraint - bowtie_c (PPP)"!(ComponentPose=@n4, Alpha=1, ConstraintWeights_0=0.0215, ConstraintWeights_1:float=0.9601, ConstraintWeights_2:float=0.0185) -> (Pose) {
    alphaBoolBlend {blendOption=Linear}
    boneToModify {boneName=bowtie_c}
    constraintSetup [{targetBone {boneName=clavicle_r}}, {targetBone {boneName=spine_05}}, {targetBone {boneName=clavicle_l}}]
    constraintWeights [1,1,1]
   }
   n6 = Get target_alpha() -> (target_alpha:double)
   n7 = "Transform (Modify) Bone - Bone: bowtie_r"(ComponentPose=@n5, Alpha=@n6, Translation="-1,-2,0", Rotation="0,-20,12", Scale="1,1,1") {
    rotationMode=BMM_Additive
    rotationSpace=BCS_BoneSpace
    translationMode=BMM_Additive
    translationSpace=BCS_BoneSpace
    alphaBoolBlend {blendOption=Linear}
    alphaScaleBiasClamp {bInterpResult=true, interpSpeedDecreasing=5}
    boneToModify {boneName=bowtie_r}
    rotation {roll=12, yaw=-20}
    translation=<-1 -2>
   }
   n8 = "Transform (Modify) Bone - Bone: bowtie_l"(ComponentPose=@n7, Alpha=@n6, Translation="0,0,0", Rotation="0,-20,-12", Scale="1,1,1") {
    rotationMode=BMM_Additive
    rotationSpace=BCS_BoneSpace
    translationMode=BMM_Additive
    translationSpace=BCS_BoneSpace
    alphaBoolBlend {blendOption=Linear}
    alphaScaleBiasClamp {bInterpResult=true, interpSpeedDecreasing=5}
    boneToModify {boneName=bowtie_l}
    rotation {roll=-12, yaw=-20}
   }
   n9 = Anim Dynamics - Bone: bowtie_2_l(ComponentPose=@n8, Alpha=1, ExternalForce="0,0,0") {
    angularDampingOverride=1
    angularSpringConstant=3000
    bAngularSpring=true
    bGravityOverrideInSimSpace=true
    bOverrideAngularDamping=true
    gravityScale=0.35
    simulationSpace=RootRelative
    alphaBoolBlend {blendOption=Linear}
    alphaScaleBiasClamp {scale=0.5}
    boundBone {boneName=bowtie_2_l}
    componentLinearAccScale {z=0.6}
    componentLinearVelScale {z=0.6}
    physicsBodyDefinitions [{
      sphereCollisionRadius=0
      boundBone {boneName=bowtie_2_l}
      boxExtents=<10 20 20>
      constraintSetup {angularLimitsMax=<0 30 0>, angularLimitsMin=<0 -30 -45>, angularTarget=<1 0 0>}
      localJointOffset=<8 0 0>
     }]
   }
   n10 = Anim Dynamics - Bone: bowtie_2_r(ComponentPose=@n9, Alpha=1, ExternalForce="0,0,0") {
    angularDampingOverride=1
    angularSpringConstant=3000
    bAngularSpring=true
    bGravityOverrideInSimSpace=true
    bOverrideAngularDamping=true
    gravityScale=0.35
    simulationSpace=RootRelative
    alphaBoolBlend {blendOption=Linear}
    alphaScaleBiasClamp {scale=0.5}
    boundBone {boneName=bowtie_2_r}
    componentLinearAccScale {z=0.6}
    componentLinearVelScale {z=0.6}
    physicsBodyDefinitions [{
      sphereCollisionRadius=0
      boundBone {boneName=bowtie_2_r}
      boxExtents=<10 20 20>
      constraintSetup {angularLimitsMax=<0 30 0>, angularLimitsMin=<0 -30 -45>, angularTarget=<1 0 0>}
      localJointOffset=<9 0 0>
     }]
   }
   n11 = Anim Dynamics - Bone: bowtie_1_l(ComponentPose=@n10, Alpha=1, ExternalForce="0,0,0") {
    angularDampingOverride=1
    angularSpringConstant=3000
    bAngularSpring=true
    bGravityOverrideInSimSpace=true
    bOverrideAngularDamping=true
    gravityScale=0.35
    simulationSpace=RootRelative
    alphaBoolBlend {blendOption=Linear}
    alphaScaleBiasClamp {scale=0.5}
    boundBone {boneName=bowtie_1_l}
    componentLinearAccScale {z=0.6}
    componentLinearVelScale {z=0.6}
    physicsBodyDefinitions [{
      sphereCollisionRadius=0
      boundBone {boneName=bowtie_1_l}
      boxExtents=<10 20 20>
      constraintSetup {angularLimitsMax=<0 30 0>, angularLimitsMin=<0 -30 -45>, angularTarget=<1 0 0>}
      localJointOffset=<5 0 0>
     }]
   }
   n12 = Anim Dynamics - Bone: bowtie_1_r(ComponentPose=@n11, Alpha=1, ExternalForce="0,0,0") {
    angularDampingOverride=1
    angularSpringConstant=3000
    bAngularSpring=true
    bGravityOverrideInSimSpace=true
    bOverrideAngularDamping=true
    gravityScale=0.35
    simulationSpace=RootRelative
    alphaBoolBlend {blendOption=Linear}
    alphaScaleBiasClamp {scale=0.5}
    boundBone {boneName=bowtie_1_r}
    componentLinearAccScale {z=0.6}
    componentLinearVelScale {z=0.6}
    physicsBodyDefinitions [{
      sphereCollisionRadius=0
      boundBone {boneName=bowtie_1_r}
      boxExtents=<10 20 20>
      constraintSetup {angularLimitsMax=<0 30 0>, angularLimitsMin=<0 -30 -45>, angularTarget=<1 0 0>}
      localJointOffset=<6 0 0>
     }]
   }
   n13 = Anim Dynamics - Bone: grenade_1_r(ComponentPose=@n12, Alpha=1, ExternalForce="0,0,0") {
    angularDampingOverride=1
    angularSpringConstant=1500
    bAngularSpring=true
    bGravityOverrideInSimSpace=true
    bOverrideAngularDamping=true
    gravityScale=0.35
    simulationSpace=RootRelative
    alphaBoolBlend {blendOption=Linear}
    boundBone {boneName=grenade_1_r}
    physicsBodyDefinitions [{
      sphereCollisionRadius=0
      boundBone {boneName=grenade_1_r}
      boxExtents=<10 20 20>
      constraintSetup {angularLimitsMax=<0 0 0>, angularLimitsMin=<0 0 -40>, angularTarget=<0 0 0>}
      localJointOffset=<10 0 0>
     }]
   }
   n14 = Anim Dynamics - Bone: grenade_2_r(ComponentPose=@n13, Alpha=1, ExternalForce="0,0,0") {
    angularDampingOverride=1
    angularSpringConstant=1500
    bAngularSpring=true
    bGravityOverrideInSimSpace=true
    bOverrideAngularDamping=true
    gravityScale=0.35
    simulationSpace=RootRelative
    alphaBoolBlend {blendOption=Linear}
    boundBone {boneName=grenade_2_r}
    physicsBodyDefinitions [{
      sphereCollisionRadius=0
      boundBone {boneName=grenade_2_r}
      boxExtents=<10 20 20>
      constraintSetup {angularLimitsMax=<0 0 0>, angularLimitsMin=<0 0 -40>, angularTarget=<0 0 0>}
      localJointOffset=<10 0 0>
     }]
   }
   n15 = "Constraint - shldrLoop_r_zero (PPPPPP)"!(
    ComponentPose=@n14
    Alpha=1
    ConstraintWeights_0=0.0155
    ConstraintWeights_1:float=0.0352
    ConstraintWeights_2:float=0.109
    ConstraintWeights_3:float=0.0152
    ConstraintWeights_4:float=0.142
    ConstraintWeights_5:float=0.6832
   ) -> (Pose) {
    alphaBoolBlend {blendOption=Linear}
    boneToModify {boneName=shldrLoop_r_zero}
    constraintSetup [
     {targetBone {boneName=spine_01}}
     {targetBone {boneName=spine_02}}
     {targetBone {boneName=spine_03}}
     {targetBone {boneName=spine_04}}
     {targetBone {boneName=spine_05}}
     {targetBone {boneName=clavicle_r}}
    ]
    constraintWeights [1,1,1,1,1,1]
   }
   n16 = "Constraint - shldrLoop_l_zero (PPP)"!(ComponentPose=@n15, Alpha=1, ConstraintWeights_0=0.0826, ConstraintWeights_1:float=0.1575, ConstraintWeights_2:float=0.7599) -> (Pose) {
    alphaBoolBlend {blendOption=Linear}
    boneToModify {boneName=shldrLoop_l_zero}
    constraintSetup [{targetBone {boneName=spine_03}}, {targetBone {boneName=spine_05}}, {targetBone {boneName=clavicle_l}}]
    constraintWeights [1,1,1]
   }
   n17 = Anim Dynamics - Bone: shldrLoop_l(ComponentPose=@n16, Alpha=1, ExternalForce="0,0,0") {
    angularDampingOverride=1
    angularSpringConstant=1500
    bAngularSpring=true
    bGravityOverrideInSimSpace=true
    bOverrideAngularDamping=true
    gravityScale=0.35
    simulationSpace=RootRelative
    alphaBoolBlend {blendOption=Linear}
    boundBone {boneName=shldrLoop_l}
    componentLinearAccScale {z=0.7}
    componentLinearVelScale {z=0.7}
    physicsBodyDefinitions [{
      sphereCollisionRadius=0
      boundBone {boneName=shldrLoop_l}
      boxExtents=<10 20 20>
      constraintSetup {angularLimitsMax=<0 0 85>, angularLimitsMin=<0 0 -85>, angularTarget=<0 0 0>}
      localJointOffset=<10 0 0>
     }]
   }
   n18 = Anim Dynamics - Bone: shldrLoop_r(ComponentPose=@n17, Alpha=1, ExternalForce="0,0,0") {
    angularDampingOverride=1
    angularSpringConstant=1500
    bAngularSpring=true
    bGravityOverrideInSimSpace=true
    bOverrideAngularDamping=true
    gravityScale=0.35
    simulationSpace=RootRelative
    alphaBoolBlend {blendOption=Linear}
    boundBone {boneName=shldrLoop_r}
    componentLinearAccScale {z=0.7}
    componentLinearVelScale {z=0.7}
    physicsBodyDefinitions [{
      sphereCollisionRadius=0
      boundBone {boneName=shldrLoop_r}
      boxExtents=<10 20 20>
      constraintSetup {angularLimitsMax=<0 0 85>, angularLimitsMin=<0 0 -85>, angularTarget=<0 0 0>}
      localJointOffset=<10 0 0>
     }]
   }
   n19 = Anim Dynamics - Bone: shoeTongue_l(ComponentPose=@n18, Alpha=1, ExternalForce="0,0,0") {
    angularDampingOverride=1
    angularSpringConstant=1500
    bAngularSpring=true
    bGravityOverrideInSimSpace=true
    bOverrideAngularDamping=true
    gravityScale=0
    simulationSpace=RootRelative
    alphaBoolBlend {blendOption=Linear}
    alphaScaleBiasClamp {scale=0.4}
    boundBone {boneName=shoeTongue_l}
    physicsBodyDefinitions [{
      sphereCollisionRadius=0
      boundBone {boneName=shoeTongue_l}
      boxExtents=<10 20 20>
      constraintSetup {angularLimitsMax=<0 0 40>, angularLimitsMin=<0 0 -40>, angularTarget=<1 0 0>}
      localJointOffset=<20 0 0>
     }]
   }
   n20 = Anim Dynamics - Bone: shoeTongue_r(ComponentPose=@n19, Alpha=1, ExternalForce="0,0,0") {
    angularDampingOverride=1
    angularSpringConstant=1500
    bAngularSpring=true
    bGravityOverrideInSimSpace=true
    bOverrideAngularDamping=true
    gravityScale=0
    simulationSpace=RootRelative
    alphaBoolBlend {blendOption=Linear}
    alphaScaleBiasClamp {scale=0.4}
    boundBone {boneName=shoeTongue_r}
    physicsBodyDefinitions [{
      sphereCollisionRadius=0
      boundBone {boneName=shoeTongue_r}
      boxExtents=<10 20 20>
      constraintSetup {angularLimitsMax=<0 0 40>, angularLimitsMin=<0 0 -40>, angularTarget=<1 0 0>}
      localJointOffset=<20 0 0>
     }]
   }
   Output Pose(Result=Component To Local(ComponentPose=@n20)) {layerGroup=None}
  ]}
 EventGraph {nodes [
   Event OnBeginTargeting() {is=title K2Node_Event}
   > Set target_alpha(target_alpha:double=1) -> (Output_Get:double)
   Event OnEndTargeting() {is=title K2Node_Event}
   > Set target_alpha(target_alpha=0)
   Event BlueprintUpdateAnimation()
  ]}
]
variables [target_alpha {type=number}]
