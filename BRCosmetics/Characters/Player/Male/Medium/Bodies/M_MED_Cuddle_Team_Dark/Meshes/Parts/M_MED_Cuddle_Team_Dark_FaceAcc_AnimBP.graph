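# Post-follow AnimBP for the Cuddle Team Dark face accessory mask, copying its attached parent mesh's pose and adding secondary jiggle.
# A Constraint node drives mask_zero from weighted head and neck bones.
# Chained spring-based AnimDynamics nodes, each scaled to partial alpha, then sway both ears, three hair tufts, the left pupil, jaw and tongue in root-relative space.
# The event graph holds only unconnected default nodes, with no logic or variables.
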AnimBlueprint /BRCosmetics/Characters/Player/Male/Medium/Bodies/M_MED_Cuddle_Team_Dark/Meshes/Parts/M_MED_Cuddle_Team_Dark_FaceAcc_AnimBP
target_skeleton=&/M_MED_Cuddle_Team_Dark_FaceAcc_Skeleton
graphs [
 AnimGraph {nodes [
   n1 = "Constraint - mask_zero (PPP)"!(
    ComponentPose=Local To Component(LocalPose=Copy Pose From Mesh() {bUseAttachedParent=true})
    Alpha=1
    ConstraintWeights_0=0.56
    ConstraintWeights_1:float=0.074
    ConstraintWeights_2:float=0.366
   ) -> (Pose) {
    alphaBoolBlend {blendOption=Linear}
    boneToModify {boneName=mask_zero}
    constraintSetup [{targetBone {boneName=head}}, {targetBone {boneName=neck_01}}, {targetBone {boneName=neck_02}}]
    constraintWeights [1,1,1]
   }
   n2 = Anim Dynamics - Bone: ear_r(ComponentPose=@n1, Alpha=1, ExternalForce="0,0,0") {
    angularDampingOverride=1
    angularSpringConstant=1500
    bAngularSpring=true
    bGravityOverrideInSimSpace=true
    bOverrideAngularDamping=true
    gravityScale=0.35
    simulationSpace=RootRelative
    alphaBoolBlend {blendOption=Linear}
    alphaScaleBiasClamp {scale=0.2}
    boundBone {boneName=ear_r}
    physicsBodyDefinitions [{
      sphereCollisionRadius=0
      boundBone {boneName=ear_r}
      boxExtents=<10 20 20>
      constraintSetup {angularLimitsMax=<0 50 80>, angularLimitsMin=<0 -50 -80>, angularTarget=<1 0 0>}
      localJointOffset=<10 0 0>
     }]
   }
   n3 = Anim Dynamics - Bone: ear_l(ComponentPose=@n2, Alpha=1, ExternalForce="0,0,0") {
    angularDampingOverride=1
    angularSpringConstant=1500
    bAngularSpring=true
    bGravityOverrideInSimSpace=true
    bOverrideAngularDamping=true
    gravityScale=0.35
    simulationSpace=RootRelative
    alphaBoolBlend {blendOption=Linear}
    alphaScaleBiasClamp {scale=0.2}
    boundBone {boneName=ear_l}
    physicsBodyDefinitions [{
      sphereCollisionRadius=0
      boundBone {boneName=ear_l}
      boxExtents=<10 20 20>
      constraintSetup {angularLimitsMax=<0 50 80>, angularLimitsMin=<0 -50 -80>, angularTarget=<1 0 0>}
      localJointOffset=<10 0 0>
     }]
   }
   n4 = Anim Dynamics - Bone: hairTuft_1(ComponentPose=@n3, Alpha=1, ExternalForce="0,0,0") {
    angularDampingOverride=1
    angularSpringConstant=1500
    bAngularSpring=true
    bGravityOverrideInSimSpace=true
    bOverrideAngularDamping=true
    gravityScale=0.1
    simulationSpace=RootRelative
    alphaBoolBlend {blendOption=Linear}
    alphaScaleBiasClamp {scale=0.5}
    boundBone {boneName=hairTuft_1}
    physicsBodyDefinitions [{
      sphereCollisionRadius=0
      boundBone {boneName=hairTuft_1}
      boxExtents=<10 20 20>
      constraintSetup {angularLimitsMax=<0 20 40>, angularLimitsMin=<0 -20 -40>, angularTarget=<1 0 0>}
      localJointOffset=<10 0 0>
     }]
   }
   n5 = Anim Dynamics - Bone: hairTuft_2(ComponentPose=@n4, Alpha=1, ExternalForce="0,0,0") {
    angularDampingOverride=1
    angularSpringConstant=1500
    bAngularSpring=true
    bGravityOverrideInSimSpace=true
    bOverrideAngularDamping=true
    gravityScale=0.1
    simulationSpace=RootRelative
    alphaBoolBlend {blendOption=Linear}
    alphaScaleBiasClamp {scale=0.5}
    boundBone {boneName=hairTuft_2}
    physicsBodyDefinitions [{
      sphereCollisionRadius=0
      boundBone {boneName=hairTuft_2}
      boxExtents=<10 20 20>
      constraintSetup {angularLimitsMax=<0 20 40>, angularLimitsMin=<0 -20 -40>, angularTarget=<1 0 0>}
      localJointOffset=<10 0 0>
     }]
   }
   n6 = Anim Dynamics - Bone: hairTuft_3(ComponentPose=@n5, Alpha=1, ExternalForce="0,0,0") {
    angularDampingOverride=1
    angularSpringConstant=1500
    bAngularSpring=true
    bGravityOverrideInSimSpace=true
    bOverrideAngularDamping=true
    gravityScale=0.1
    simulationSpace=RootRelative
    alphaBoolBlend {blendOption=Linear}
    alphaScaleBiasClamp {scale=0.5}
    boundBone {boneName=hairTuft_3}
    physicsBodyDefinitions [{
      sphereCollisionRadius=0
      boundBone {boneName=hairTuft_3}
      boxExtents=<10 20 20>
      constraintSetup {angularLimitsMax=<0 20 40>, angularLimitsMin=<0 -20 -40>, angularTarget=<1 0 0>}
      localJointOffset=<10 0 0>
     }]
   }
   n7 = Anim Dynamics - Bone: pupil_l(ComponentPose=@n6, Alpha=1, ExternalForce="0,0,0") {
    angularDampingOverride=1
    angularSpringConstant=1500
    bAngularSpring=true
    bGravityOverrideInSimSpace=true
    bOverrideAngularDamping=true
    gravityScale=0.35
    simulationSpace=RootRelative
    alphaBoolBlend {blendOption=Linear}
    alphaScaleBiasClamp {scale=0.15}
    boundBone {boneName=pupil_l}
    physicsBodyDefinitions [{
      sphereCollisionRadius=0
      boundBone {boneName=pupil_l}
      boxExtents=<10 20 20>
      constraintSetup {angularLimitsMax=<0 20 40>, angularLimitsMin=<0 -20 -40>, angularTarget=<0 0 0>}
      localJointOffset=<10 0 0>
     }]
   }
   n8 = Anim Dynamics - Bone: jaw_c(ComponentPose=@n7, Alpha=1, ExternalForce="0,0,0") {
    angularDampingOverride=1
    angularSpringConstant=1500
    bAngularSpring=true
    bGravityOverrideInSimSpace=true
    bOverrideAngularDamping=true
    gravityScale=0.35
    simulationSpace=RootRelative
    alphaBoolBlend {blendOption=Linear}
    alphaScaleBiasClamp {scale=0.3}
    boundBone {boneName=jaw_c}
    physicsBodyDefinitions [{
      sphereCollisionRadius=0
      boundBone {boneName=jaw_c}
      boxExtents=<20 10 20>
      constraintSetup {angularTargetAxis=AxisY, angularLimitsMax=<5 0 0>, angularLimitsMin=<-5 0 -20>, angularTarget=<0 1 0>}
      localJointOffset=<0 15 0>
     }]
   }
   n9 = Anim Dynamics - Bone: tongue(ComponentPose=@n8, Alpha=1, ExternalForce="0,0,0") {
    angularDampingOverride=1
    angularSpringConstant=1500
    bAngularSpring=true
    bGravityOverrideInSimSpace=true
    bOverrideAngularDamping=true
    gravityScale=0.35
    simulationSpace=RootRelative
    alphaBoolBlend {blendOption=Linear}
    alphaScaleBiasClamp {scale=0.3}
    boundBone {boneName=tongue}
    physicsBodyDefinitions [{
      sphereCollisionRadius=0
      boundBone {boneName=tongue}
      boxExtents=<10 20 20>
      constraintSetup {angularLimitsMax=<0 15 0>, angularLimitsMin=<0 -15 -40>, angularTarget=<0 0 0>}
      localJointOffset=<10 0 0>
     }]
   }
   Output Pose(Result=Component To Local(ComponentPose=@n9)) {layerGroup=None}
  ]}
 EventGraph {nodes [
   TryGetPawnOwner()
   Event BlueprintUpdateAnimation()
  ]}
]
variables []
